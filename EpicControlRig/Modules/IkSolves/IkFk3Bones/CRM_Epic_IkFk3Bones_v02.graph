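# Modular Control Rig module that builds an IK/FK switchable setup for a three-segment limb (Bones A-D).
# Construction resolves connectors, then spawns FK controls, an IK control, an auto pole vector, an optional end roll control and animation channels.
# The roll control's parent null aims from Bone A, blending out its rest rotation offset as the limb extends.
# Helper functions match IK and FK controls to the bones and toggle their visibility.
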ControlRigBlueprint /EpicControlRig/Modules/IkSolves/IkFk3Bones/CRM_Epic_IkFk3Bones_v02
graphs [
 ConstructionGraph {nodes [
   RigUnit_PrepareForExecution = &RigUnit_PrepareForExecution()
   /Script/ControlRig.RigUnit_SetupShapeLibraryFromUserData(LibraryName=CRSL, LogShapeLibraries=False, NameSpace=CRSL, Path=ShapeLibrary, ExecutePin=@RigUnit_PrepareForExecution.ExecutePin) {is=struct RigVMUnitNode}
   > RigVMFunction_Sequence = RigVMAggregateNode()
   n1 = &RigVMFunction_ControlFlowBranch(ExecuteContext=@RigVMFunction_Sequence.C, Condition=RigVMVariableNode() {variable=End Roll Control}.Value)
   Compute Transform Aim at Root_1_1 = RigVMFunctionReferenceNode(Target Space=RigVMVariableNode() {variable=Bone C}.Value)
   n2 = &RigUnit_HierarchyAddNull(
    Name=IK Roll
    Space=GlobalSpace
    Transform=@"Compute Transform Aim at Root_1_1".Result
    ExecutePin=@n1.True
    Parent=RigVMVariableNode() {variable=IK Control}.Value
   )
   > RigVMVariableNode(Value=@n2.Item) {variable=IK Roll Aim Null}
   n3 = RigVMVariableNode() {variable=IK Roll Control Shape}
   Control Color Override From Metadata v01_2 = RigVMFunctionReferenceNode(Color Override=@n3.Value.Color, Module Color=RigVMVariableNode() {variable=Color}.Value)
   Get Control Shape Name From Item v01_3 = RigVMFunctionReferenceNode(
    ShapeLib Namespace=CRSL
    Item=RigVMDispatchNode(Index=-2, Array=RigVMVariableNode() {variable=FK Bones}.Value) {template="DISPATCH_RigVMDispatch_ArrayGetAtIndex(in Array,in Index,out Element)"}.Element
    Default Shape=@n3.Value.Name
   )
   > n4 = /Script/ControlRig.RigUnit_HierarchyAddControlRotator(
    InitialValue=(Pitch=0,Yaw=0,Roll=0)
    Name=IK Roll
    OffsetSpace=LocalSpace
    OffsetTransform=(Rotation=(X=0,Y=0,Z=0,W=1),Translation=(X=0,Y=0,Z=0),Scale3D=(X=1,Y=1,Z=1))
    Settings="(InitialSpace=LocalSpace,Limits=(LimitPitch=(bMinimum=False,bMaximum=False),LimitYaw=(bMinimum=False,bMaximum=False),LimitRoll=(bMinimum=False,bMaximum=False),MinValue=(Pitch=-180,Yaw=-180,Roll=-180),MaxValue=(Pitch=180,Yaw=180,Roll=180),bDrawLimits=True), ...(truncated)"
    Settings.Shape.Color=@"Control Color Override From Metadata v01_2".Result
    Settings.Shape.Name=@"Get Control Shape Name From Item v01_3".Result
    Settings.Shape.Transform=@n3.Value.Transform
    Settings.Shape.bVisible=@n3.Value.bVisible
    Parent=RigVMVariableNode() {variable=IK Roll Aim Null}.Value
   ) {
    is=struct RigVMUnitNode
    template="SpawnControl::Execute(in InitialValue,in Settings,in OffsetTransform,in OffsetSpace,in Parent,in Name,out Item)"
   }
   > RigVMVariableNode(Value=@n4.Item) {variable=IK Roll Control}
   > Set Mirror Behavior_2 = RigVMFunctionReferenceNode(Value=RigVMVariableNode() {variable=Mirror Behavior}.Value, Item=@n4.Item)
   n5 = RigVMVariableNode() {variable=Bone C}
   > Set_Metadata_1 = RigVMDispatchNode(
    Name=IkRollDelta
    NameSpace=Self
    Value=&RigVMFunction_MathTransformMakeRelative(
     Global=&RigUnit_GetTransform(bInitial=true, Item=@n5.Value).Transform
     Parent=&RigUnit_GetTransform(bInitial=true, Item=@n4.Item).Transform
    ).Local
    Item=@n4.Item
   ) {template="DISPATCH_RigDispatch_SetMetadata(in Item,in Name,in NameSpace,in Value,out Success)"}
   n6 = &RigUnit_HierarchyAddNull(
    Name=IK Effector
    Space=GlobalSpace
    Transform=&RigUnit_GetTransform(bInitial=False, Item=RigVMVariableNode() {variable=Bone C}.Value).Transform
    Parent=RigVMDispatchNode(
     True=RigVMVariableNode() {variable=IK Roll Control}.Value
     Condition=RigVMVariableNode() {variable=End Roll Control}.Value
     False=RigVMVariableNode() {variable=IK Control}.Value
    ) {template="DISPATCH_RigVMDispatch_If(in Condition,in True,in False,out Result)"}.Result
    ExecutePin=@n1.Completed
   )
   > RigVMVariableNode(Value=@n6.Item) {variable=IK Effector Null}
   DISPATCH_RigVMDispatch_ArrayReset = RigVMDispatchNode(ExecuteContext=@RigVMFunction_Sequence.E, Array=RigVMVariableNode() {variable=FK Controls}.Value) {template="DISPATCH_RigVMDispatch_ArrayReset(io Array)"}
   > DISPATCH_RigVMDispatch_ArrayIterator = RigVMDispatchNode(Array=RigVMVariableNode() {variable=FK Bones}.Value) {template="DISPATCH_RigVMDispatch_ArrayIterator(in Array,out Element,out Index,out Count,out Ratio)"}
   Int to Name = /Script/RigVM.RigVMFunction_MathIntToName(PaddedSize=0, Number=@DISPATCH_RigVMDispatch_ArrayIterator.Index) {is=struct RigVMUnitNode, template="Int to Name::Execute(in Number,in PaddedSize,out Result)"}
   Concat = /Script/RigVM.RigVMFunction_NameConcat(A="FK ", B=@"Int to Name".Result) {is=struct RigVMUnitNode, template="Concat::Execute(in A,in B,out Result)"}
   Get Control Color From Metadata = RigVMFunctionReferenceNode(Color Override=RigVMVariableNode() {variable=Color}.Value)
   Get Control Shape Name From Item v01_4 = RigVMFunctionReferenceNode(
    ShapeLib Namespace=CRSL
    Item=@DISPATCH_RigVMDispatch_ArrayIterator.Element
    Default Shape=RigVMVariableNode() {variable=FKs Control Shape}.Value.Name
   )
   n7 = &RigUnit_GetTransform(bInitial=true, Item=@DISPATCH_RigVMDispatch_ArrayIterator.Element)
   n8 = RigVMVariableNode() {variable=Fk Controls Rotation Offset}
   Multiply = /Script/RigVM.RigVMFunction_MathQuaternionMul(A=@n7.Transform.Rotation, B=@n8.Value) {is=struct RigVMUnitNode, template="Multiply::Execute(in A,in B,out Result)"}
   RigVMFunction_MathTransformMake = /Script/RigVM.RigVMFunction_MathTransformMake(Scale=@n7.Transform.Scale3D, Translation=@n7.Transform.Translation, Rotation=@Multiply.Result) {is=struct RigVMUnitNode}
   Inverse = /Script/RigVM.RigVMFunction_MathQuaternionInverse(Value=@n8.Value) {is=struct RigVMUnitNode, template="Inverse::Execute(in Value,out Result)"}
   > n9 = &RigUnit_HierarchyAddControlRotator(
    InitialValue=(Pitch=0,Yaw=0,Roll=0)
    OffsetSpace=GlobalSpace
    Settings="(InitialSpace=LocalSpace,Limits=(LimitPitch=(bMinimum=False,bMaximum=False),LimitYaw=(bMinimum=False,bMaximum=False),LimitRoll=(bMinimum=False,bMaximum=False),MinValue=(Pitch=-180,Yaw=-180,Roll=-180),MaxValue=(Pitch=180,Yaw=180,Roll=180),bDrawLimits=True), ...(truncated)"
    Name=@Concat.Result
    Settings.Shape.Color=@"Get Control Color From Metadata".Result
    Settings.Shape.Name=@"Get Control Shape Name From Item v01_4".Result
    OffsetTransform=@RigVMFunction_MathTransformMake.Result
    Settings.Shape.Transform=&RigVMFunction_MathTransformMake(Scale=(X=1,Y=1,Z=1), Translation=(X=0,Y=0,Z=0), Rotation=@Inverse.Result).Result
    Parent=RigVMVariableNode() {variable=FK Parent}.Value
   ) {template="SpawnControl::Execute(in InitialValue,in Settings,in OffsetTransform,in OffsetSpace,in Parent,in Name,out Item)"}
   > RigVMVariableNode(Value=@n9.Item) {variable=FK Parent}
   > DISPATCH_RigVMDispatch_ArrayAdd = RigVMDispatchNode(Array=RigVMVariableNode() {variable=FK Controls}.Value, Element=@n9.Item) {template="DISPATCH_RigVMDispatch_ArrayAdd(io Array,in Element,out Index)"}
   > Set Mirror Axis_1 = RigVMFunctionReferenceNode(Item=(Type=Bone,Name="None"), Value=RigVMVariableNode() {variable=Mirror Axis}.Value)
   > Set Mirror Behavior_1 = RigVMFunctionReferenceNode(Value=RigVMVariableNode() {variable=Mirror Behavior}.Value, Item=@n9.Item)
   Get Distances Between = RigVMFunctionReferenceNode(
    Initial=False
    Items=((Type=None,Name="None"),(Type=None,Name="None"),(Type=None,Name="None"),(Type=None,Name="None"))
    Items.1=RigVMVariableNode() {variable=Bone B}.Value
    Items.2=RigVMVariableNode() {variable=Bone C}.Value
    Items.3=RigVMVariableNode() {variable=Bone D}.Value
    Items.0=RigVMVariableNode() {variable=Bone A}.Value
    ExecuteContext=@RigVMFunction_Sequence.H
   )
   > Get Distances Between_1 = RigVMFunctionReferenceNode(
    Initial=False
    Items=((Type=None,Name="None"),(Type=None,Name="None"))
    Items.0=RigVMVariableNode() {variable=Bone A}.Value
    Items.1=RigVMVariableNode() {variable=Bone B}.Value
   )
   ArraySum = /Script/RigVM.RigVMFunction_MathDoubleArraySum(Array=@"Get Distances Between".Distances) {is=struct RigVMUnitNode, template="ArraySum::Execute(in Array,out Sum)"}
   > Pythagoras = RigVMFunctionReferenceNode(
    B=@ArraySum.Sum
    A=&RigVMFunction_MathDoubleArraySum(Array=@"Get Distances Between_1".Distances) {template="ArraySum::Execute(in Array,out Sum)"}.Sum
   )
   > RigVMVariableNode(Value=@Pythagoras.Hypotenus) {variable=Length}
   n10 = RigVMVariableNode() {variable=Bone A}
   > RigVMVariableNode(Value=&RigVMFunction_MathVectorDistance(
     A=&RigUnit_GetTransform(bInitial=true, Item=@n10.Value).Transform.Translation
     B=RigVMVariableNode() {variable=Effector Location}.Value
    ).Result) {variable=Rest Distance}
   RigVMVariableNode(
    Value=/Script/ControlRig.RigUnit_ResolveConnector(Connector=(Type=Connector,Name="Bone A"), SkipSocket=False) {is=struct RigVMUnitNode}.Result
    ExecuteContext=@RigVMFunction_Sequence.A
   ) {variable=Bone A}
   > RigVMVariableNode(Value=&RigUnit_ResolveConnector(Connector=(Type=Connector,Name="Bone B"), SkipSocket=False).Result) {variable=Bone B}
   > RigVMVariableNode(Value=&RigUnit_ResolveConnector(Connector=(Type=Connector,Name="Bone C"), SkipSocket=False).Result) {variable=Bone C}
   > RigVMVariableNode(Value=&RigUnit_ResolveConnector(Connector=(Type=Connector,Name="Bone D"), SkipSocket=False).Result) {variable=Bone D}
   > n11 = /Script/ControlRig.RigUnit_HierarchyAddBone(
    Name=End
    Space=LocalSpace
    Transform=(Rotation=(X=0,Y=0,Z=0,W=1),Translation=(X=0,Y=0,Z=0),Scale3D=(X=1,Y=1,Z=1))
    Transform.Translation=RigVMVariableNode() {variable=Effector Offset}.Value
    Parent=RigVMVariableNode() {variable=Bone D}.Value
   ) {is=struct RigVMUnitNode}
   > RigVMVariableNode(Value=&RigUnit_GetTransform(bInitial=False, Item=@n11.Item).Transform.Translation) {variable=Effector Location}
   > RigVMVariableNode(Value=&RigUnit_ResolveConnector(Connector=(Type=Connector,Name="IK Parent"), SkipSocket=False).Result) {variable=IK Parent}
   Make_Array_1 = RigVMDispatchNode(
    Values=((Type=None,Name="None"),((),Type=None,Name="None"),((),Type=None,Name="None"),((),Type=None,Name="None"))
    Values.1=RigVMVariableNode() {variable=Bone B}.Value
    Values.0=RigVMVariableNode() {variable=Bone A}.Value
    Values.2=RigVMVariableNode() {variable=Bone C}.Value
    Values.3=RigVMVariableNode() {variable=Bone D}.Value
   ) {template="DISPATCH_RigVMDispatch_ArrayMake(in Values,out Array)"}
   > RigVMVariableNode(Value=@Make_Array_1.Array) {variable=FK Bones}
   > RigVMVariableNode(Value=(Type=Connector,Name="FK Parent")) {variable=FK Parent Connector}
   > RigVMVariableNode(Value=RigVMVariableNode() {variable=FK Parent Connector}.Value) {variable=FK Parent}
   n12 = &RigUnit_GetTransform(
    bInitial=true
    Item=RigVMDispatchNode(Index=0, Array=RigVMVariableNode() {variable=FK Bones}.Value) {template="DISPATCH_RigVMDispatch_ArrayGetAtIndex(in Array,in Index,out Element)"}.Element
   )
   n13 = RigVMVariableNode() {variable=Fk Controls Rotation Offset}
   n14 = &RigUnit_HierarchyAddNull(
    Name=Default Space
    Space=GlobalSpace
    Transform=&RigVMFunction_MathTransformMake(
     Scale=@n12.Transform.Scale3D
     Translation=@n12.Transform.Translation
     Rotation=&RigVMFunction_MathQuaternionMul(A=@n12.Transform.Rotation, B=@n13.Value) {template="Multiply::Execute(in A,in B,out Result)"}.Result
    ).Result
    Parent=RigVMVariableNode() {variable=FK Parent}.Value
    ExecutePin=@RigVMFunction_Sequence.D
   )
   > RigVMVariableNode(Value=@n14.Item) {variable=FK Parent}
   n15 = &RigVMFunction_NameConcat(B=" Orient", A=&RigUnit_ResolveConnector(Connector=(Type=Connector,Name="FK Orient Space"), SkipSocket=False).Result.Name) {template="Concat::Execute(in A,in B,out Result)"}
   n16 = &RigUnit_GetTransform(
    bInitial=true
    Item=RigVMDispatchNode(Index=0, Array=RigVMVariableNode() {variable=FK Bones}.Value) {template="DISPATCH_RigVMDispatch_ArrayGetAtIndex(in Array,in Index,out Element)"}.Element
   )
   n17 = RigVMVariableNode() {variable=Fk Controls Rotation Offset}
   n18 = &RigUnit_HierarchyAddNull(
    Space=GlobalSpace
    Name=@n15.Result
    Parent=&RigUnit_ResolveConnector(SkipSocket=False, Connector=RigVMVariableNode() {variable=FK Parent Connector}.Value).Result
    Transform=&RigVMFunction_MathTransformMake(
     Scale=@n16.Transform.Scale3D
     Translation=@n16.Transform.Translation
     Rotation=&RigVMFunction_MathQuaternionMul(A=@n16.Transform.Rotation, B=@n17.Value) {template="Multiply::Execute(in A,in B,out Result)"}.Result
    ).Result
    ExecutePin=@RigVMFunction_Sequence.F
   )
   > RigVMVariableNode(Value=@n18.Item) {variable=FK Orient Space Null}
   > /Script/ControlRig.RigUnit_AddAvailableSpaces(
    Spaces=((Key=(Type=Control,Name="None"),Label="None"))
    Control=RigVMDispatchNode(Index=0, Array=RigVMVariableNode() {variable=FK Controls}.Value) {template="DISPATCH_RigVMDispatch_ArrayGetAtIndex(in Array,in Index,out Element)"}.Element
    Spaces.0.Label=@n15.Result
    Spaces.0.Key=@n18.Item
   ) {is=struct RigVMUnitNode}
   n19 = RigVMVariableNode() {variable=IK Control Shape}
   Control Color Override From Metadata v01_3 = RigVMFunctionReferenceNode(Color Override=@n19.Value.Color, Module Color=RigVMVariableNode() {variable=Color}.Value)
   n20 = RigVMVariableNode() {variable=Bone D}
   Get Control Shape Name From Item v01_1 = RigVMFunctionReferenceNode(ShapeLib Namespace=CRSL, Default Shape=@n19.Value.Name, Item=@n20.Value)
   SpawnControl = /Script/ControlRig.RigUnit_HierarchyAddControlTransform(
    InitialValue=(Rotation=(X=0,Y=0,Z=0,W=1),Translation=(X=0,Y=0,Z=0),Scale3D=(X=1,Y=1,Z=1))
    Name=IK
    OffsetSpace=GlobalSpace
    OffsetTransform=(Rotation=(X=0,Y=0,Z=0,W=1),Translation=(X=0,Y=0,Z=0),Scale3D=(X=1,Y=1,Z=1))
    Settings="(InitialSpace=LocalSpace,bUsePreferredRotationOrder=False,PreferredRotationOrder=YZX,Limits=(LimitTranslationX=(bMinimum=False,bMaximum=False),LimitTranslationY=(bMinimum=False,bMaximum=False),LimitTranslationZ=(bMinimum=False,bMaximum=False),LimitPitch=(b ...(truncated)"
    Settings.Shape.Color=@"Control Color Override From Metadata v01_3".Result
    Settings.Shape.Name=@"Get Control Shape Name From Item v01_1".Result
    OffsetTransform.Translation=&RigUnit_GetTransform(bInitial=true, Item=@n20.Value).Transform.Translation
    Settings.Shape.Transform=@n19.Value.Transform
    Settings.Shape.bVisible=@n19.Value.bVisible
    Parent=RigVMVariableNode() {variable=IK Parent}.Value
    ExecutePin=@RigVMFunction_Sequence.B
   ) {
    is=struct RigVMUnitNode
    template="SpawnControl::Execute(in InitialValue,in Settings,in OffsetTransform,in OffsetSpace,in Parent,in Name,out Item)"
   }
   > RigVMVariableNode(Value=@SpawnControl.Item) {variable=IK Control}
   n21 = RigVMVariableNode() {variable=Bone D}
   > Set_Metadata = RigVMDispatchNode(
    Name=IkDelta
    NameSpace=Self
    Value=&RigVMFunction_MathTransformMakeRelative(
     Global=&RigUnit_GetTransform(bInitial=true, Item=@n21.Value).Transform
     Parent=&RigUnit_GetTransform(bInitial=true, Item=@SpawnControl.Item).Transform
    ).Local
    Item=@SpawnControl.Item
   ) {template="DISPATCH_RigDispatch_SetMetadata(in Item,in Name,in NameSpace,in Value,out Success)"}
   n22 = RigVMVariableNode() {variable=IK Control}
   > Set_Metadata_2 = RigVMDispatchNode(
    Name=ShapeInitialTransform
    NameSpace=Self
    Value=/Script/ControlRig.RigUnit_GetShapeTransform(Control=@n22.Value.Name) {is=struct RigVMUnitNode}.Transform
    Item=@n22.Value
   ) {template="DISPATCH_RigDispatch_SetMetadata(in Item,in Name,in NameSpace,in Value,out Success)"}
   n23 = RigVMVariableNode() {variable=IK Control}
   > Set Mirror Axis = RigVMFunctionReferenceNode(Value=(X=0,Y=1,Z=1), Item=@n23.Value)
   > Set Mirror Behavior = RigVMFunctionReferenceNode(Value=false, Item=@n23.Value)
   Control Color Override From Metadata v01 = RigVMFunctionReferenceNode(Color Override=RigVMVariableNode() {variable=PV Control Shape}.Value.Color, Module Color=RigVMVariableNode() {variable=Color}.Value)
   > RigVMVariableNode(Value=@"Control Color Override From Metadata v01".Result) {variable=PV Control Color}
   n24 = RigVMVariableNode() {variable=PV Control Shape}
   n25 = RigVMVariableNode() {variable=Bone A}
   > Construct Auto Pole Vector v02_1 = RigVMFunctionReferenceNode(
    Initial=true
    Name=PV
    Shape=(bVisible=True,Name="Default",Color=(R=1,G=0,B=0,A=1),Transform=(Rotation=(X=0,Y=0,Z=0,W=1),Translation=(X=0,Y=0,Z=0),Scale3D=(X=1,Y=1,Z=1)))
    Shape.Color=@"Control Color Override From Metadata v01".Result
    Parent=RigVMVariableNode() {variable=IK Parent}.Value
    IK Control=RigVMVariableNode() {variable=IK Control}.Value
    Shape.Name=@n24.Value.Name
    Shape.Transform=@n24.Value.Transform
    Shape.bVisible=@n24.Value.bVisible
    Offset Multiplier=RigVMVariableNode() {variable=PV Distance Mult}.Value
    Item A=@n25.Value
    PV Root Space=@n25.Value
    Item B=RigVMVariableNode() {variable=Bone B}.Value
    Item C=RigVMVariableNode() {variable=Bone C}.Value
   )
   > RigVMVariableNode(Value=@"Construct Auto Pole Vector v02_1".Pv Control) {variable=PV Control}
   > RigVMVariableNode(Value=@"Construct Auto Pole Vector v02_1".Pv Root Space Null) {variable=PV Root Space Null}
   > &RigUnit_AddAvailableSpaces(
    Spaces=((Key=(Type=Control,Name="None"),Label="None"))
    Spaces.0.Key=RigVMDispatchNode(Default=(Type=Connector,Name="IK Parent"), Name=OffsetControl, NameSpace=Root) {template="DISPATCH_RigDispatch_GetModuleMetadata(in Name,in NameSpace,in Default,out Value,out Found)"}.Value
    Control=RigVMVariableNode() {variable=PV Control}.Value
   )
   > &RigUnit_AddAvailableSpaces(
    Spaces=((Key=(Type=Control,Name="None"),Label="None"))
    Spaces.0.Key=RigVMVariableNode() {variable=IK Control}.Value
    Control=RigVMVariableNode() {variable=PV Control}.Value
   )
   Compute Transform Aim at Root = RigVMFunctionReferenceNode(Target Space=RigVMVariableNode() {variable=Bone A}.Value)
   Compute Transform Aim at Root_1 = RigVMFunctionReferenceNode(Target Space=RigVMVariableNode() {variable=Bone C}.Value)
   > RigVMVariableNode(Value=/Script/RigVM.RigVMFunction_MathQuaternionMakeRelative(Parent=@"Compute Transform Aim at Root".Result.Rotation, Global=@"Compute Transform Aim at Root_1".Result.Rotation) {is=struct RigVMUnitNode, template="Make Relative::Execute(in Global,in Parent,out Local)"}.Local) {variable=Rotation Offset}
   Make_Array = RigVMDispatchNode(
    Values=((Type=None,Name="None"),((),Type=None,Name="None"),((),Type=None,Name="None"),((),Type=None,Name="None"))
    Values.0=RigVMVariableNode() {variable=Bone A}.Value
    Values.1=RigVMVariableNode() {variable=Bone B}.Value
    Values.2=RigVMVariableNode() {variable=Bone C}.Value
    Values.3=RigVMVariableNode() {variable=Bone D}.Value
   ) {template="DISPATCH_RigVMDispatch_ArrayMake(in Values,out Array)"}
   Get Chain Length = RigVMFunctionReferenceNode(Initial=true, Items=@Make_Array.Array, ExecuteContext=@RigVMFunction_Sequence.I)
   Clone = RigVMDispatchNode(Array=RigVMVariableNode() {variable=FK Controls}.Value) {template="DISPATCH_RigVMDispatch_ArrayClone(in Array,out Clone)"}
   > n26 = RigVMDispatchNode(
    Other=((Type=None,Name="None"),(Type=None,Name="None"),(Type=None,Name="None"))
    Array=@Clone.Clone
    Other.2=RigVMVariableNode() {variable=IK Roll Control}.Value
    Other.0=RigVMVariableNode() {variable=IK Control}.Value
    Other.1=RigVMVariableNode() {variable=PV Control}.Value
   ) {template="DISPATCH_RigVMDispatch_ArrayUnion(io Array,in Other)"}
   Get_Module_Metadata_1 = RigVMDispatchNode(Default=1, Name=Global Control Scale, NameSpace=Root) {template="DISPATCH_RigDispatch_GetModuleMetadata(in Name,in NameSpace,in Default,out Value,out Found)"}
   > Scale Control Shape v01_1 = RigVMFunctionReferenceNode(
    Auto Scale Factor=1
    Global Scale=@Get_Module_Metadata_1.Value
    Controls=@n26.Array
    Scale=RigVMVariableNode() {variable=Control Scale}.Value
   )
   SpawnAnimationChannel = /Script/ControlRig.RigUnit_HierarchyAddAnimationChannelBool(
    InitialValue=true
    LimitsEnabled=()
    MaximumValue=True
    MinimumValue=False
    Name=IK FK Switch
    Parent=RigVMVariableNode() {variable=IK Control}.Value
    ExecutePin=@RigVMFunction_Sequence.G
   ) {
    is=struct RigVMUnitNode
    template="SpawnAnimationChannel::Execute(in InitialValue,in MinimumValue,in MaximumValue,in LimitsEnabled,in Parent,in Name,out Item)"
   }
   > RigVMVariableNode(Value=@SpawnAnimationChannel.Item) {variable=IK FK Channel}
   > /Script/ControlRig.RigUnit_SetChannelHosts(Channel=@SpawnAnimationChannel.Item, Hosts=RigVMVariableNode() {variable=FK Controls}.Value) {is=struct RigVMUnitNode}
   n27 = RigVMVariableNode() {variable=PV Control}
   > RigVMDispatchNode(Array=(), Element=@n27.Value) {template="DISPATCH_RigVMDispatch_ArrayAdd(io Array,in Element,out Index)"}
   > n28 = &RigUnit_HierarchyAddAnimationChannelBool(InitialValue=true, LimitsEnabled=(), MaximumValue=True, MinimumValue=False, Name=Draw Pv Line, Parent=@n27.Value) {template="SpawnAnimationChannel::Execute(in InitialValue,in MinimumValue,in MaximumValue,in LimitsEnabled,in Parent,in Name,out Item)"}
   Make_Array_4 = RigVMDispatchNode(Values=((Type=None,Name="None")), Values.0=RigVMVariableNode() {variable=IK Control}.Value) {template="DISPATCH_RigVMDispatch_ArrayMake(in Values,out Array)"}
   > &RigUnit_SetChannelHosts(Hosts=@Make_Array_4.Array, Channel=@n28.Item)
   > n29 = /Script/ControlRig.RigUnit_HierarchyAddAnimationChannelFloat(
    LimitsEnabled=(Enabled=(bMinimum=True,bMaximum=True))
    MaximumValue=1
    MinimumValue=0
    Name=PV Twist Follow
    InitialValue=RigVMVariableNode() {variable=PV Twist Follow}.Value
    Parent=RigVMVariableNode() {variable=IK Control}.Value
   ) {
    is=struct RigVMUnitNode
    template="SpawnAnimationChannel::Execute(in InitialValue,in MinimumValue,in MaximumValue,in LimitsEnabled,in Parent,in Name,out Item)"
   }
   > RigVMVariableNode(Value=@n29.Item) {variable=PV Twist Follow Channel}
   Make_Array_3 = RigVMDispatchNode(Values=((Type=None,Name="None")), Values.0=RigVMVariableNode() {variable=PV Control}.Value) {template="DISPATCH_RigVMDispatch_ArrayMake(in Values,out Array)"}
   > &RigUnit_SetChannelHosts(Channel=RigVMVariableNode() {variable=IK FK Channel}.Value, Hosts=@Make_Array_3.Array)
   > &RigUnit_SetChannelHosts(Channel=RigVMVariableNode() {variable=PV Twist Follow Channel}.Value, Hosts=@Make_Array_3.Array)
   > n30 = &RigUnit_HierarchyAddAnimationChannelBool(
    InitialValue=true
    LimitsEnabled=()
    MaximumValue=True
    MinimumValue=False
    Name=Draw Pv
    Parent=RigVMVariableNode() {variable=PV Control}.Value
   ) {template="SpawnAnimationChannel::Execute(in InitialValue,in MinimumValue,in MaximumValue,in LimitsEnabled,in Parent,in Name,out Item)"}
   > RigVMVariableNode(Value=@n30.Item) {variable=Draw Pv Channel}
   Make_Array_2 = RigVMDispatchNode(Values=((Type=None,Name="None")), Values.0=RigVMVariableNode() {variable=IK Control}.Value) {template="DISPATCH_RigVMDispatch_ArrayMake(in Values,out Array)"}
   > n31 = &RigUnit_SetChannelHosts(Hosts=@Make_Array_2.Array, Channel=@n30.Item)
   n32 = &RigVMFunction_ControlFlowBranch(ExecuteContext=@n31.ExecutePin, Condition=RigVMVariableNode() {variable=End Roll Control}.Value)
   RigVMFunction_ControlFlowBranch = &RigVMFunction_ControlFlowBranch(ExecuteContext=@n32.Completed, Condition=RigVMVariableNode() {variable=IK FK Manipulation}.Value)
   For_Each = RigVMDispatchNode(ExecuteContext=@RigVMFunction_ControlFlowBranch.True, Array=RigVMVariableNode() {variable=FK Controls}.Value) {template="DISPATCH_RigVMDispatch_ArrayIterator(in Array,out Element,out Index,out Count,out Ratio)"}
   > /Script/ControlRig.RigUnit_SetControlDrivenList(
    Driven=((Type=None,Name="None"),(Type=None,Name="None"),(Type=None,Name="None"))
    Control=@For_Each.Element.Name
    Driven.0=RigVMVariableNode() {variable=IK Control}.Value
    Driven.1=RigVMVariableNode() {variable=PV Control}.Value
    Driven.2=RigVMVariableNode() {variable=IK Roll Control}.Value
   ) {is=struct RigVMUnitNode, template="SetControlDrivenList::Execute(in Control,in Driven)"}
   For_Each_1 = RigVMDispatchNode(
    Array=((Type=None,Name="None"),(Type=None,Name="None"),(Type=None,Name="None"))
    ExecuteContext=@For_Each.Completed
    Array.0=RigVMVariableNode() {variable=IK Control}.Value
    Array.1=RigVMVariableNode() {variable=PV Control}.Value
    Array.2=RigVMVariableNode() {variable=IK Roll Control}.Value
   ) {template="DISPATCH_RigVMDispatch_ArrayIterator(in Array,out Element,out Index,out Count,out Ratio)"}
   > &RigUnit_SetControlDrivenList(Control=@For_Each_1.Element.Name, Driven=RigVMVariableNode() {variable=FK Controls}.Value) {template="SetControlDrivenList::Execute(in Control,in Driven)"}
   Make_Array_5 = RigVMDispatchNode(Values=((Type=None,Name="None")), Values.0=RigVMVariableNode() {variable=IK Roll Control}.Value) {template="DISPATCH_RigVMDispatch_ArrayMake(in Values,out Array)"}
   &RigUnit_SetChannelHosts(Hosts=@Make_Array_5.Array, ExecutePin=@n32.True, Channel=RigVMVariableNode() {variable=IK FK Channel}.Value)
   Num = RigVMDispatchNode(Array=@Make_Array.Array) {template="DISPATCH_RigVMDispatch_ArrayGetNum(in Array,out Num)"}
   RigVMDispatchNode(
    False=1
    True=&RigVMFunction_MathDoubleDiv(B=200, A=@"Get Chain Length".Length).Result
    Condition=/Script/RigVM.RigVMFunction_MathIntGreater(B=1, A=@Num.Num) {is=struct RigVMUnitNode, template="Greater::Execute(in A,in B,out Result)"}.Result
   ) {template="DISPATCH_RigVMDispatch_If(in Condition,in True,in False,out Result)"}
  ]}
 ContainedGraph {nodes [
   n1 = RigVMFunctionEntryNode()
   SequenceAggregate = &RigVMFunction_Sequence(ExecuteContext=@n1.ExecuteContext)
   DISPATCH_RigVMDispatch_ArrayReset = RigVMDispatchNode(ExecuteContext=@SequenceAggregate.A, Array=RigVMVariableNode() {variable=Values}.Value) {template="DISPATCH_RigVMDispatch_ArrayReset(io Array)"}
   n2 = RigVMVariableNode() {variable=Items}
   > DISPATCH_RigVMDispatch_ArrayIterator = RigVMDispatchNode(Array=@n2.Value) {template="DISPATCH_RigVMDispatch_ArrayIterator(in Array,out Element,out Index,out Count,out Ratio)"}
   DISPATCH_RigVMDispatch_CoreEquals = RigVMDispatchNode(B=0, A=@DISPATCH_RigVMDispatch_ArrayIterator.Index) {template="DISPATCH_RigVMDispatch_CoreEquals(in A,in B,out Result)"}
   RigVMFunction_ControlFlowBranch = &RigVMFunction_ControlFlowBranch(ExecuteContext=@DISPATCH_RigVMDispatch_ArrayIterator.ExecuteContext, Condition=@DISPATCH_RigVMDispatch_CoreEquals.Result)
   n3 = RigVMVariableNode() {variable=Initial}
   Subtract = /Script/RigVM.RigVMFunction_MathIntSub(B=1, A=@DISPATCH_RigVMDispatch_ArrayIterator.Index) {is=struct RigVMUnitNode, template="Subtract::Execute(in A,in B,out Result)"}
   DISPATCH_RigVMDispatch_ArrayGetAtIndex = RigVMDispatchNode(Index=@Subtract.Result, Array=@n2.Value) {template="DISPATCH_RigVMDispatch_ArrayGetAtIndex(in Array,in Index,out Element)"}
   RigVMFunction_MathVectorDistance = &RigVMFunction_MathVectorDistance(
    A=&RigUnit_GetTransform(Item=@DISPATCH_RigVMDispatch_ArrayIterator.Element, bInitial=@n3.Value).Transform.Translation
    B=&RigUnit_GetTransform(Item=@DISPATCH_RigVMDispatch_ArrayGetAtIndex.Element, bInitial=@n3.Value).Transform.Translation
   )
   DISPATCH_RigVMDispatch_ArrayAdd = RigVMDispatchNode(
    Array=@DISPATCH_RigVMDispatch_ArrayReset.Array
    ExecuteContext=@RigVMFunction_ControlFlowBranch.False
    Element=@RigVMFunction_MathVectorDistance.Result
   ) {template="DISPATCH_RigVMDispatch_ArrayAdd(io Array,in Element,out Index)"}
   RigVMFunctionReturnNode(Distances=@DISPATCH_RigVMDispatch_ArrayReset.Array, ExecuteContext=@SequenceAggregate.B)
   RigVMVariableNode() {variable=Values}
   RigVMVariableNode() {variable=Values}
  ]}
 New Function_ContainedGraph {nodes [
   n1 = RigVMFunctionEntryNode()
   RigVMFunction_Sequence = &RigVMFunction_Sequence(ExecuteContext=@n1.ExecuteContext)
   n2 = &RigVMFunction_Sequence(ExecuteContext=@RigVMFunction_Sequence.A)
   DISPATCH_RigVMDispatch_ArrayIterator = RigVMDispatchNode(Array=RigVMVariableNode() {variable=IK Controls}.Value, ExecuteContext=@n2.A) {template="DISPATCH_RigVMDispatch_ArrayIterator(in Array,out Element,out Index,out Count,out Ratio)"}
   > SetControlVisibility = /Script/ControlRig.RigUnit_SetControlVisibility(Item=@DISPATCH_RigVMDispatch_ArrayIterator.Element, bVisible=@n1.IK Solve) {is=struct RigVMUnitNode, template="SetControlVisibility::Execute(in Item,in Pattern,in bVisible)"}
   n3 = RigVMDispatchNode(Array=RigVMVariableNode() {variable=FK Controls}.Value, ExecuteContext=@n2.B) {template="DISPATCH_RigVMDispatch_ArrayIterator(in Array,out Element,out Index,out Count,out Ratio)"}
   RigVMFunction_MathBoolNot = &RigVMFunction_MathBoolNot(Value=@n1.IK Solve)
   > &RigUnit_SetControlVisibility(Item=@n3.Element, bVisible=@RigVMFunction_MathBoolNot.Result) {template="SetControlVisibility::Execute(in Item,in Pattern,in bVisible)"}
   RigVMFunctionReturnNode(ExecuteContext=@RigVMFunction_Sequence.B)
   RigVMVariableNode() {variable=IK Solve}
   RigVMVariableNode() {variable=IK Solve}
  ]}
 New Function_ContainedGraph {nodes [
   n1 = RigVMFunctionEntryNode()
   Multiply = &RigVMFunction_MathDoubleMul(A=@n1.A, B=@n1.A)
   Add = &RigVMFunction_MathDoubleAdd(A=@Multiply.Result, B=&RigVMFunction_MathDoubleMul(A=@n1.B, B=@n1.B).Result)
   Sqrt = /Script/RigVM.RigVMFunction_MathDoubleSqrt(Value=@Add.Result) {is=struct RigVMUnitNode, template="Sqrt::Execute(in Value,out Result)"}
   > RigVMFunctionReturnNode(Hypotenus=@Sqrt.Result)
  ]}
 New Function_ContainedGraph {nodes [
   n1 = RigVMFunctionEntryNode()
   n2 = /Script/ControlRig.RigUnit_AimBoneMath(
    DebugSettings=(bEnabled=False,Scale=10,WorldOffset=(Rotation=(X=0,Y=0,Z=0,W=1),Translation=(X=0,Y=0,Z=0),Scale3D=(X=1,Y=1,Z=1)))
    Primary=(Weight=1,Axis=(X=1,Y=0,Z=0),Target=(X=0,Y=0,Z=0),Kind=Location,Space=(Type=Bone,Name="None"))
    Secondary=(Weight=1,Axis=(X=0,Y=0,Z=1),Target=(X=0,Y=0,Z=0),Kind=Location,Space=(Type=Bone,Name="None"))
    Weight=1
    Primary.Space=@n1.Target Space
    InputTransform=&RigUnit_GetTransform(bInitial=false, Item=RigVMVariableNode() {variable=IK Control}.Value).Transform
    Secondary.Space=RigVMVariableNode() {variable=PV Control}.Value
    Primary.Axis=RigVMVariableNode() {variable=Primary Axis}.Value
    Secondary.Axis=RigVMVariableNode() {variable=Secondary Axis}.Value
   ) {is=struct RigVMUnitNode}
   > RigVMFunctionReturnNode(Result=@n2.Result)
  ]}
 New Function_ContainedGraph {nodes [
   RigVMFunctionEntryNode()
   Compute Transform Aim at Root = RigVMFunctionReferenceNode(Target Space=RigVMVariableNode() {variable=Bone A}.Value)
   n1 = RigVMVariableNode() {variable=Bone A}
   n2 = RigVMVariableNode() {variable=IK Control}
   Remap = &RigVMFunction_MathFloatRemap(
    TargetMinimum=0
    bClamp=true
    Value=&RigVMFunction_MathVectorDistance(
     A=&RigUnit_GetTransform(bInitial=False, Item=@n1.Value).Transform.Translation
     B=&RigUnit_GetTransform(bInitial=False, Item=@n2.Value).Transform.Translation
    ).Result
    SourceMaximum=RigVMVariableNode() {variable=Length}.Value
    SourceMinimum=RigVMVariableNode() {variable=Rest Distance}.Value
   )
   n3 = RigVMVariableNode() {variable=Rotation Offset}
   > &RigUnit_SetTransform(
    Weight=1
    bInitial=False
    Value=/Script/RigVM.RigVMFunction_MathTransformMakeAbsolute(
     Local=(Rotation=(X=0,Y=0,Z=0,W=1),Translation=(X=0,Y=0,Z=0),Scale3D=(X=1,Y=1,Z=1))
     Parent=@"Compute Transform Aim at Root".Result
     Local.Rotation=/Script/RigVM.RigVMFunction_MathQuaternionSlerp(B=(X=0,Y=0,Z=0,W=1), T=@Remap.Result, A=@n3.Value) {is=struct RigVMUnitNode, template="Interpolate::Execute(in A,in B,in T,out Result)"}.Result
    ) {is=struct RigVMUnitNode, template="Make Absolute::Execute(in Local,in Parent,out Global)"}.Global
    Item=RigVMVariableNode() {variable=IK Roll Aim Null}.Value
   )
   > RigVMFunctionReturnNode()
  ]}
 New Function_ContainedGraph {nodes [
   RigVMFunctionEntryNode()
   > DISPATCH_RigVMDispatch_ArrayIterator = RigVMDispatchNode(Array=RigVMVariableNode() {variable=FK Bones}.Value) {template="DISPATCH_RigVMDispatch_ArrayIterator(in Array,out Element,out Index,out Count,out Ratio)"}
   > Set Transform = &RigUnit_SetRotation(
    Space=GlobalSpace
    Weight=1
    bInitial=False
    Item=RigVMDispatchNode(Array=RigVMVariableNode() {variable=FK Controls}.Value, Index=@DISPATCH_RigVMDispatch_ArrayIterator.Index) {template="DISPATCH_RigVMDispatch_ArrayGetAtIndex(in Array,in Index,out Element)"}.Element
    Value=&RigVMFunction_MathQuaternionMul(
     A=&RigUnit_GetTransform(bInitial=False, Item=@DISPATCH_RigVMDispatch_ArrayIterator.Element).Transform.Rotation
     B=RigVMVariableNode() {variable=Fk Controls Rotation Offset}.Value
    ) {template="Multiply::Execute(in A,in B,out Result)"}.Result
   )
   RigVMFunctionReturnNode(ExecuteContext=@DISPATCH_RigVMDispatch_ArrayIterator.Completed)
  ]}
 New Function_ContainedGraph {nodes [
   RigVMFunctionEntryNode()
   n1 = RigVMVariableNode() {variable=Bone D}
   n2 = RigVMVariableNode() {variable=IK Control}
   Get_Metadata = RigVMDispatchNode(Default=(Rotation=(X=0,Y=0,Z=0,W=1),Translation=(X=0,Y=0,Z=0),Scale3D=(X=1,Y=1,Z=1)), Name=IkDelta, NameSpace=Self, Item=@n2.Value) {template="DISPATCH_RigDispatch_GetMetadata(in Item,in Name,in NameSpace,in Default,out Value,out Found)"}
   Inverse = /Script/RigVM.RigVMFunction_MathTransformInverse(Value=@Get_Metadata.Value) {is=struct RigVMUnitNode, template="Inverse::Execute(in Value,out Result)"}
   > &RigUnit_SetTransform(
    Weight=1
    bInitial=False
    Value=&RigVMFunction_MathTransformMakeAbsolute(Parent=&RigUnit_GetTransform(bInitial=False, Item=@n1.Value).Transform, Local=@Inverse.Result) {template="Make Absolute::Execute(in Local,in Parent,out Global)"}.Global
    Item=@n2.Value
   )
   > Compute Auto Pole Vector v02_1 = RigVMFunctionReferenceNode(
    PV Orient Space=RigVMDispatchNode(Default=(Type=Connector,Name="IK Parent"), Name=OffsetControl, NameSpace=Root) {template="DISPATCH_RigDispatch_GetModuleMetadata(in Name,in NameSpace,in Default,out Value,out Found)"}.Value
    PV Twist Follow=/Script/ControlRig.RigUnit_GetFloatAnimationChannelFromItem(bInitial=False, Item=RigVMVariableNode() {variable=PV Twist Follow Channel}.Value) {is=struct RigVMUnitNode, template="GetAnimationChannelFromItem::Execute(out Value,in Item,in bInitial)"}.Value
    PV Root Space=RigVMVariableNode() {variable=PV Root Space Null}.Value
    PV Control=RigVMVariableNode() {variable=PV Control}.Value
    IK Control=RigVMVariableNode() {variable=IK Control}.Value
   )
   Compute Pole Vector Location v01 = RigVMFunctionReferenceNode(
    Initial=false
    Item A=RigVMVariableNode() {variable=Bone A}.Value
    Up Vector=RigVMVariableNode() {variable=Secondary Axis}.Value
    OffsetMultiplier=RigVMVariableNode() {variable=PV Distance Mult}.Value
    Item B=RigVMVariableNode() {variable=Bone B}.Value
   )
   > SetTransform_4_1 = /Script/ControlRig.RigUnit_SetTranslation(
    Space=GlobalSpace
    Weight=1
    bInitial=false
    bPropagateToChildren=True
    Value=@"Compute Pole Vector Location v01".Result
    Item=RigVMVariableNode() {variable=PV Control}.Value
   ) {is=struct RigVMUnitNode, template="Set Transform::Execute(in Item,in Space,in bInitial,in Value,in Weight,in bPropagateToChildren)"}
   RigVMFunction_ControlFlowBranch = &RigVMFunction_ControlFlowBranch(ExecuteContext=@SetTransform_4_1.ExecutePin, Condition=RigVMVariableNode() {variable=End Roll Control}.Value)
   RigVMFunctionReturnNode(ExecuteContext=@RigVMFunction_ControlFlowBranch.Completed)
   Compute Roll Aim = RigVMFunctionReferenceNode(ExecuteContext=@RigVMFunction_ControlFlowBranch.True)
   n3 = RigVMVariableNode() {variable=Bone C}
   n4 = RigVMVariableNode() {variable=IK Roll Control}
   Get_Metadata_2 = RigVMDispatchNode(Default=(Rotation=(X=0,Y=0,Z=0,W=1),Translation=(X=0,Y=0,Z=0),Scale3D=(X=1,Y=1,Z=1)), Name=IkRollDelta, NameSpace=Self, Item=@n4.Value) {template="DISPATCH_RigDispatch_GetMetadata(in Item,in Name,in NameSpace,in Default,out Value,out Found)"}
   > &RigUnit_SetTransform(
    Weight=1
    bInitial=False
    Value=&RigVMFunction_MathTransformMakeAbsolute(
     Parent=&RigUnit_GetTransform(bInitial=False, Item=@n3.Value).Transform
     Local=&RigVMFunction_MathTransformInverse(Value=@Get_Metadata_2.Value) {template="Inverse::Execute(in Value,out Result)"}.Result
    ) {template="Make Absolute::Execute(in Local,in Parent,out Global)"}.Global
    Item=@n4.Value
   )
  ]}
 New Function_ContainedGraph {nodes [
   RigVMFunctionEntryNode()
   n1 = RigVMVariableNode() {variable=FK Bones}
   > RigVMFunction_DebugLineStripNoSpace = /Script/RigVM.RigVMFunction_DebugLineStripNoSpace(
    Color=(R=1,G=1,B=1,A=1)
    DebugDrawSettings=(DepthPriority=SDPG_Foreground,Lifetime=-1)
    Points=((X=0,Y=0,Z=0),(X=0,Y=0,Z=0))
    Thickness=1
    WorldOffset=(Rotation=(X=0,Y=0,Z=0,W=1),Translation=(X=0,Y=0,Z=0),Scale3D=(X=1,Y=1,Z=1))
    bEnabled=true
    Points.0=&RigUnit_GetTransform(
     bInitial=False
     Item=RigVMDispatchNode(Index=1, Array=@n1.Value) {template="DISPATCH_RigVMDispatch_ArrayGetAtIndex(in Array,in Index,out Element)"}.Element
    ).Transform.Translation
    Points.1=&RigUnit_GetTransform(bInitial=False, Item=RigVMVariableNode() {variable=PV Control}.Value).Transform.Translation
   ) {is=struct RigVMUnitNode}
   > RigVMFunctionReturnNode()
  ]}
 RigVMFunctionLibrary {nodes [
   Compute Roll Aim = RigVMCollapseNode(ExecuteContext=())
   Compute Transform Aim at Root = RigVMCollapseNode(Target Space=(Type=Bone,Name="None"))
   Draw PV Line = RigVMCollapseNode(ExecuteContext=())
   Get Distances Between = RigVMCollapseNode(ExecuteContext=(), Initial=False, Items=())
   Hide Show Controls = RigVMCollapseNode(ExecuteContext=(), FK Controls=(), IK Controls=(), IK Solve=false)
   Match FK Controls = RigVMCollapseNode(ExecuteContext=())
   Match IK Controls = RigVMCollapseNode(ExecuteContext=())
   Pythagoras = RigVMCollapseNode(A=1, B=1, ExecuteContext=())
  ]}
 RigVMFunction_Sequence_1_1_ContainedGraph {nodes [
   n1 = RigVMFunctionEntryNode()
   n2 = &RigVMFunction_Sequence(ExecuteContext=@n1.ExecuteContext)
   n3 = &RigVMFunction_Sequence(ExecuteContext=@n2.B)
   n4 = &RigVMFunction_Sequence(ExecuteContext=@n3.B)
   RigVMFunctionReturnNode(A=@n2.A, B=@n3.A, C=@n4.A, D=@n4.B)
  ]}
 RigVMFunction_Sequence_1_ContainedGraph {nodes [
   n1 = RigVMFunctionEntryNode()
   RigVMFunction_Sequence = &RigVMFunction_Sequence(ExecuteContext=@n1.ExecuteContext)
   n2 = &RigVMFunction_Sequence(ExecuteContext=@RigVMFunction_Sequence.B)
   n3 = &RigVMFunction_Sequence(ExecuteContext=@n2.B)
   n4 = &RigVMFunction_Sequence(ExecuteContext=@n3.B)
   n5 = &RigVMFunction_Sequence(ExecuteContext=@n4.B)
   n6 = &RigVMFunction_Sequence(ExecuteContext=@n5.B)
   n7 = &RigVMFunction_Sequence(ExecuteContext=@n6.B)
   n8 = &RigVMFunction_Sequence(ExecuteContext=@n7.B)
   RigVMFunctionReturnNode(A=@RigVMFunction_Sequence.A, B=@n2.A, C=@n3.A, D=@n4.A, E=@n5.A, F=@n6.A, G=@n7.A, H=@n8.A, I=@n8.B)
  ]}
 RigVMFunction_Sequence_2_ContainedGraph {nodes [
   n1 = RigVMFunctionEntryNode()
   RigVMFunction_Sequence = &RigVMFunction_Sequence(ExecuteContext=@n1.ExecuteContext)
   n2 = &RigVMFunction_Sequence(ExecuteContext=@RigVMFunction_Sequence.B)
   RigVMFunctionReturnNode(A=@RigVMFunction_Sequence.A, B=@n2.A, C=@n2.B)
  ]}
 RigVMFunction_Sequence_4_ContainedGraph {nodes [
   n1 = RigVMFunctionEntryNode()
   n2 = &RigVMFunction_Sequence(ExecuteContext=@n1.ExecuteContext)
   n3 = &RigVMFunction_Sequence(ExecuteContext=@n2.B)
   n4 = &RigVMFunction_Sequence(ExecuteContext=@n3.B)
   n5 = &RigVMFunction_Sequence(ExecuteContext=@n4.B)
   RigVMFunctionReturnNode(B=@n3.A, A=@n2.A, C=@n4.A, D=@n5.A, E=@n5.B)
  ]}
 RigVMModel Forward Solve Graph {nodes [
   RigUnit_BeginExecution = &RigUnit_BeginExecution()
   RigVMFunction_Sequence_1 = RigVMAggregateNode(ExecuteContext=@RigUnit_BeginExecution.ExecutePin)
   RigVMFunction_ControlFlowBranch = &RigVMFunction_ControlFlowBranch(ExecuteContext=@RigVMFunction_Sequence_1.B, Condition=RigVMVariableNode() {variable=IK Solve}.Value)
   RigVMFunction_Sequence = RigVMAggregateNode(ExecuteContext=@RigVMFunction_ControlFlowBranch.True)
   n1 = &RigVMFunction_ControlFlowBranch(ExecuteContext=@RigVMFunction_Sequence.A, Condition=RigVMVariableNode() {variable=End Roll Control}.Value)
   Compute Roll Aim = RigVMFunctionReferenceNode(ExecuteContext=@n1.True)
   Compute Auto Pole Vector v02_1 = RigVMFunctionReferenceNode(
    PV Orient Space=RigVMDispatchNode(Default=(Type=Connector,Name="IK Parent"), Name=OffsetControl, NameSpace=Root) {template="DISPATCH_RigDispatch_GetModuleMetadata(in Name,in NameSpace,in Default,out Value,out Found)"}.Value
    PV Twist Follow=&RigUnit_GetFloatAnimationChannelFromItem(bInitial=False, Item=RigVMVariableNode() {variable=PV Twist Follow Channel}.Value) {template="GetAnimationChannelFromItem::Execute(out Value,in Item,in bInitial)"}.Value
    ExecuteContext=@RigVMFunction_Sequence.B
    PV Root Space=RigVMVariableNode() {variable=PV Root Space Null}.Value
    PV Control=RigVMVariableNode() {variable=PV Control}.Value
    IK Control=RigVMVariableNode() {variable=IK Control}.Value
   )
   RigUnit_GetTransform = &RigUnit_GetTransform(bInitial=False, Item=RigVMVariableNode() {variable=IK Effector Null}.Value)
   > /Script/ControlRig.RigUnit_TwoBoneIKSimplePerItem(
    DebugSettings=(bEnabled=False,Scale=10,WorldOffset=(Rotation=(X=0,Y=0,Z=0,W=1),Translation=(X=0,Y=0,Z=0),Scale3D=(X=1,Y=1,Z=1)))
    ItemALength=0
    ItemBLength=0
    PoleVectorKind=Location
    PoleVectorSpace=(Type=Bone,Name="None")
    SecondaryAxisWeight=1
    StretchMaximumRatio=1.25
    StretchStartRatio=0.75
    Weight=1
    bEnableStretch=False
    bPropagateToChildren=true
    Effector=@RigUnit_GetTransform.Transform
    PoleVector=&RigUnit_GetTransform(bInitial=False, Item=RigVMVariableNode() {variable=PV Control}.Value).Transform.Translation
    ItemA=RigVMVariableNode() {variable=Bone A}.Value
    ItemB=RigVMVariableNode() {variable=Bone B}.Value
    EffectorItem=RigVMVariableNode() {variable=Bone C}.Value
    PrimaryAxis=RigVMVariableNode() {variable=Primary Axis}.Value
    SecondaryAxis=RigVMVariableNode() {variable=Secondary Axis}.Value
   ) {is=struct RigVMUnitNode}
   > RigVMFunction_Sequence_2 = RigVMAggregateNode()
   n2 = &RigVMFunction_ControlFlowBranch(ExecuteContext=@RigVMFunction_Sequence_2.A, Condition=RigVMVariableNode() {variable=End Roll Control}.Value)
   n3 = RigVMVariableNode() {variable=Bone C}
   n4 = &RigUnit_AimBoneMath(
    DebugSettings=(bEnabled=False,Scale=10,WorldOffset=(Rotation=(X=0,Y=0,Z=0,W=1),Translation=(X=0,Y=0,Z=0),Scale3D=(X=1,Y=1,Z=1)))
    Primary=(Weight=1,Axis=(X=1,Y=0,Z=0),Target=(X=0,Y=0,Z=0),Kind=Location,Space=(Type=Bone,Name="None"))
    Secondary=(Weight=1,Axis=(X=0,Y=0,Z=-1),Target=(X=0,Y=0,Z=0),Kind=Location,Space=(Type=Bone,Name="None"))
    Weight=1
    InputTransform=&RigUnit_GetTransform(bInitial=False, Item=@n3.Value).Transform
    Secondary.Space=RigVMVariableNode() {variable=PV Control}.Value
    Primary.Space=RigVMVariableNode() {variable=IK Control}.Value
    Primary.Axis=RigVMVariableNode() {variable=Primary Axis}.Value
    Secondary.Axis=RigVMVariableNode() {variable=Secondary Axis}.Value
   )
   Set Transform_1 = &RigUnit_SetRotation(Space=GlobalSpace, Weight=1, bInitial=False, bPropagateToChildren=true, Value=@n4.Result.Rotation, Item=@n3.Value, ExecutePin=@n2.True)
   n5 = RigVMVariableNode() {variable=IK Control}
   Get_Metadata = RigVMDispatchNode(Default=(Rotation=(X=0,Y=0,Z=0,W=1),Translation=(X=0,Y=0,Z=0),Scale3D=(X=1,Y=1,Z=1)), Name=IkDelta, NameSpace=Self, Item=@n5.Value) {template="DISPATCH_RigDispatch_GetMetadata(in Item,in Name,in NameSpace,in Default,out Value,out Found)"}
   > Set Transform = &RigUnit_SetRotation(
    Space=GlobalSpace
    Weight=1
    bInitial=False
    Value=&RigVMFunction_MathTransformMakeAbsolute(Parent=&RigUnit_GetTransform(bInitial=False, Item=@n5.Value).Transform, Local=@Get_Metadata.Value) {template="Make Absolute::Execute(in Local,in Parent,out Global)"}.Global.Rotation
    Item=RigVMVariableNode() {variable=Bone D}.Value
   )
   Set Transform_1_1 = &RigUnit_SetRotation(
    Space=GlobalSpace
    Weight=1
    bInitial=False
    Value=&RigUnit_GetTransform(bInitial=False, Item=RigVMVariableNode() {variable=IK Effector Null}.Value).Transform.Rotation
    Item=RigVMVariableNode() {variable=Bone C}.Value
    ExecutePin=@n2.False
   )
   n6 = &RigVMFunction_ControlFlowBranch(
    Condition=/Script/ControlRig.RigUnit_GetBoolAnimationChannel(Channel=Draw Pv Line, bInitial=False, Control=RigVMVariableNode() {variable=PV Control}.Value.Name) {is=struct RigVMUnitNode, template="GetAnimationChannel::Execute(out Value,in Control,in Channel,in bInitial)"}.Value
    ExecuteContext=@RigVMFunction_Sequence_2.C
   )
   n7 = /Script/ControlRig.RigUnit_IsInteracting() {is=struct RigVMUnitNode}
   n8 = &RigVMFunction_ControlFlowBranch(Condition=@n7.bIsInteracting, ExecuteContext=@n6.False)
   n9 = &RigVMFunction_ControlFlowBranch(
    Condition=RigVMDispatchNode(
     Array=()
     Element=RigVMDispatchNode(Index=0, Array=@n7.Items) {template="DISPATCH_RigVMDispatch_ArrayGetAtIndex(in Array,in Index,out Element)"}.Element
    ) {template="DISPATCH_RigVMDispatch_ArrayFind(in Array,in Element,out Index,out Success)"}.Success
    ExecuteContext=@n8.True
   )
   Draw PV Line_1 = RigVMFunctionReferenceNode(ExecuteContext=@n9.True)
   Draw PV Line_0 = RigVMFunctionReferenceNode(ExecuteContext=@n6.True)
   Match FK Controls = RigVMFunctionReferenceNode(ExecuteContext=@RigVMFunction_Sequence_2.B)
   n10 = &RigVMFunction_ControlFlowBranch(
    Condition=/Script/ControlRig.RigUnit_GetBoolAnimationChannelFromItem(bInitial=False, Item=RigVMVariableNode() {variable=Draw Pv Channel}.Value) {is=struct RigVMUnitNode, template="GetAnimationChannelFromItem::Execute(out Value,in Item,in bInitial)"}.Value
    ExecuteContext=@RigVMFunction_Sequence_2.D
   )
   n11 = RigVMVariableNode() {variable=PV Control Shape}
   /Script/ControlRig.RigUnit_HierarchySetShapeSettings(
    Settings=(bVisible=true,Name="Default",Color=(R=1,G=0,B=0,A=1),Transform=(Rotation=(X=0,Y=0,Z=0,W=1),Translation=(X=0,Y=0,Z=0),Scale3D=(X=1,Y=1,Z=1)))
    ExecutePin=@n10.True
    Settings.Name=@n11.Value.Name
    Settings.Transform=@n11.Value.Transform
    Settings.bVisible=@n11.Value.bVisible
    Settings.Color=RigVMVariableNode() {variable=PV Control Color}.Value
    Item=RigVMVariableNode() {variable=PV Control}.Value
   ) {is=struct RigVMUnitNode}
   n12 = RigVMVariableNode() {variable=Bone C}
   n13 = RigVMVariableNode() {variable=PV Control}
   n14 = &RigVMFunction_MathTransformMul(
    B=&RigVMFunction_MathTransformMakeRelative(
     Global=&RigUnit_GetTransform(bInitial=False, Item=@n12.Value).Transform
     Parent=&RigUnit_GetTransform(bInitial=false, Item=@n13.Value).Transform
    ).Local
    A=&RigVMFunction_MathTransformMakeRelative(
     Parent=&RigUnit_GetTransform(bInitial=true, Item=@n12.Value).Transform
     Global=&RigUnit_GetTransform(bInitial=true, Item=@n13.Value).Transform
    ).Local
   ) {template="Multiply::Execute(in A,in B,out Result)"}
   n15 = RigVMVariableNode() {variable=PV Control Shape}
   RigVMFunction_MathTransformMake = &RigVMFunction_MathTransformMake(
    Rotation=@n14.Result.Rotation
    Translation=/Script/RigVM.RigVMFunction_MathTransformTransformVector(Transform=@n14.Result, Location=@n15.Value.Transform.Translation) {is=struct RigVMUnitNode}.Result
    Scale=@n15.Value.Transform.Scale3D
   )
   &RigUnit_HierarchySetShapeSettings(
    Settings=(bVisible=True,Name="Default",Color=(R=1,G=0,B=0,A=1),Transform=(Rotation=(X=0,Y=0,Z=0,W=1),Translation=(X=0,Y=0,Z=0),Scale3D=(X=1,Y=1,Z=1)))
    Settings.Transform=@RigVMFunction_MathTransformMake.Result
    Item=RigVMVariableNode() {variable=PV Control}.Value
    Settings.Color=RigVMVariableNode() {variable=PV Control Color}.Value
    ExecutePin=@n10.False
   )
   n16 = RigVMVariableNode() {variable=IK Control}
   Get_Metadata_1 = RigVMDispatchNode(
    Default=(Rotation=(X=0,Y=0,Z=0,W=1),Translation=(X=0,Y=0,Z=0),Scale3D=(X=1,Y=1,Z=1))
    Name=ShapeInitialTransform
    NameSpace=Self
    Item=@n16.Value
   ) {template="DISPATCH_RigDispatch_GetMetadata(in Item,in Name,in NameSpace,in Default,out Value,out Found)"}
   n17 = &RigVMFunction_ControlFlowBranch(Condition=@Get_Metadata_1.Found, ExecuteContext=@RigVMFunction_Sequence_2.E)
   n18 = &RigUnit_GetTransform(bInitial=False, Item=RigVMVariableNode() {variable=IK Control}.Value)
   n19 = RigVMDispatchNode(Index=-1, Array=RigVMVariableNode() {variable=FK Bones}.Value) {template="DISPATCH_RigVMDispatch_ArrayGetAtIndex(in Array,in Index,out Element)"}
   n20 = RigVMVariableNode() {variable=IK Control}
   n21 = &RigVMFunction_MathTransformMakeAbsolute(
    Parent=&RigUnit_GetTransform(bInitial=False, Item=@n19.Element).Transform
    Local=&RigVMFunction_MathTransformMakeRelative(
     Parent=&RigUnit_GetTransform(bInitial=true, Item=@n19.Element).Transform
     Global=&RigUnit_GetTransform(bInitial=true, Item=@n20.Value).Transform
    ).Local
   ) {template="Make Absolute::Execute(in Local,in Parent,out Global)"}
   /Script/ControlRig.RigUnit_SetShapeTransform(
    Transform=&RigVMFunction_MathTransformMakeAbsolute(Local=@Get_Metadata_1.Value, Parent=&RigVMFunction_MathTransformMakeRelative(Parent=@n18.Transform, Global=@n21.Global).Local) {template="Make Absolute::Execute(in Local,in Parent,out Global)"}.Global
    ExecutePin=@n17.True
    Control=@n16.Value.Name
   ) {is=struct RigVMUnitNode}
   n22 = &RigUnit_ResolveConnector(Connector=(Type=Connector,Name="FK Orient Space"), SkipSocket=False)
   n23 = &RigUnit_GetTransform(bInitial=true, Item=@n22.Result)
   n24 = RigVMVariableNode() {variable=FK Orient Space Null}
   n25 = &RigUnit_GetTransform(bInitial=true, Item=@n24.Value)
   Set Transform_4 = &RigUnit_SetRotation(
    Space=GlobalSpace
    Weight=1
    bInitial=False
    Value=&RigVMFunction_MathTransformMul(
     B=&RigUnit_GetTransform(bInitial=False, Item=@n22.Result).Transform
     A=&RigVMFunction_MathTransformMakeRelative(Parent=@n23.Transform, Global=@n25.Transform).Local
    ) {template="Multiply::Execute(in A,in B,out Result)"}.Result.Rotation
    Item=@n24.Value
    ExecutePin=@RigVMFunction_ControlFlowBranch.False
   )
   > n26 = RigVMDispatchNode(Array=RigVMVariableNode() {variable=FK Bones}.Value) {template="DISPATCH_RigVMDispatch_ArrayIterator(in Array,out Element,out Index,out Count,out Ratio)"}
   n27 = &RigUnit_GetTransform(
    bInitial=False
    Item=RigVMDispatchNode(Index=@n26.Index, Array=RigVMVariableNode() {variable=FK Controls}.Value) {template="DISPATCH_RigVMDispatch_ArrayGetAtIndex(in Array,in Index,out Element)"}.Element
   )
   Inverse = &RigVMFunction_MathQuaternionInverse(Value=RigVMVariableNode() {variable=Fk Controls Rotation Offset}.Value) {template="Inverse::Execute(in Value,out Result)"}
   Multiply = &RigVMFunction_MathQuaternionMul(A=@n27.Transform.Rotation, B=@Inverse.Result) {template="Multiply::Execute(in A,in B,out Result)"}
   > &RigUnit_SetTransform(
    Value=(Rotation=(X=0,Y=0,Z=0,W=1),Translation=(X=0,Y=0,Z=0),Scale3D=(X=1,Y=1,Z=1))
    Weight=1
    bInitial=False
    Item=@n26.Element
    Value.Scale3D=@n27.Transform.Scale3D
    Value.Translation=@n27.Transform.Translation
    Value.Rotation=@Multiply.Result
   )
   > Match IK Controls = RigVMFunctionReferenceNode()
   Not_Equals = RigVMDispatchNode(B=RigVMVariableNode() {variable=Previous IK Solve}.Value, A=RigVMVariableNode() {variable=IK Solve}.Value) {template="DISPATCH_RigVMDispatch_CoreNotEquals(in A,in B,out Result)"}
   RigVMFunction_MathBoolNot = &RigVMFunction_MathBoolNot(Value=RigVMVariableNode() {variable=IK FK Manipulation}.Value)
   RigVMFunction_MathBoolAnd = &RigVMFunction_MathBoolAnd(B=@Not_Equals.Result, A=@RigVMFunction_MathBoolNot.Result)
   n28 = &RigVMFunction_ControlFlowBranch(Condition=@RigVMFunction_MathBoolAnd.Result, ExecuteContext=@RigVMFunction_Sequence_1.C)
   RigVMVariableNode(ExecuteContext=@n28.Completed, Value=RigVMVariableNode() {variable=IK Solve}.Value) {variable=Previous IK Solve}
   Clone = RigVMDispatchNode(Array=RigVMVariableNode() {variable=FK Controls}.Value) {template="DISPATCH_RigVMDispatch_ArrayClone(in Array,out Clone)"}
   n29 = RigVMDispatchNode(
    Array=@Clone.Clone
    Other=RigVMDispatchNode(
     Values=((Type=None,Name="None"),((),Type=None,Name="None"),((),Type=None,Name="None"))
     Values.0=RigVMVariableNode() {variable=IK Control}.Value
     Values.2=RigVMVariableNode() {variable=PV Control}.Value
     Values.1=RigVMVariableNode() {variable=IK Roll Control}.Value
    ) {template="DISPATCH_RigVMDispatch_ArrayMake(in Values,out Array)"}.Array
    ExecuteContext=@n28.True
   ) {template="DISPATCH_RigVMDispatch_ArrayAppend(io Array,in Other)"}
   > For_Each = RigVMDispatchNode(Array=@n29.Array) {template="DISPATCH_RigVMDispatch_ArrayIterator(in Array,out Element,out Index,out Count,out Ratio)"}
   > SendEvent = /Script/ControlRig.RigUnit_SendEvent(Event=RequestAutoKey, OffsetInSeconds=0, bEnable=True, bOnlyDuringInteraction=True, Item=@For_Each.Element) {is=struct RigVMUnitNode, template="SendEvent::Execute(in Event,in Item,in OffsetInSeconds,in bEnable,in bOnlyDuringInteraction)"}
   n30 = RigVMVariableNode(
    Value=&RigUnit_GetBoolAnimationChannelFromItem(bInitial=False, Item=RigVMVariableNode() {variable=IK FK Channel}.Value) {template="GetAnimationChannelFromItem::Execute(out Value,in Item,in bInitial)"}.Value
    ExecuteContext=@RigVMFunction_Sequence_1.A
   ) {variable=IK Solve}
   n31 = &RigVMFunction_ControlFlowBranch(ExecuteContext=@n30.ExecuteContext, Condition=RigVMVariableNode() {variable=IK FK Manipulation}.Value)
   n32 = &RigUnit_IsInteracting()
   n33 = &RigVMFunction_ControlFlowBranch(Condition=@n32.bIsInteracting, ExecuteContext=@n31.True)
   n34 = RigVMDispatchNode(Index=0, Array=@n32.Items) {template="DISPATCH_RigVMDispatch_ArrayGetAtIndex(in Array,in Index,out Element)"}
   n35 = &RigVMFunction_ControlFlowBranch(
    Condition=RigVMDispatchNode(Array=RigVMVariableNode() {variable=FK Controls}.Value, Element=@n34.Element) {template="DISPATCH_RigVMDispatch_ArrayFind(in Array,in Element,out Index,out Success)"}.Success
    ExecuteContext=@n33.True
   )
   n36 = &RigVMFunction_ControlFlowBranch(
    Condition=RigVMDispatchNode(
     Array=((Type=None,Name="None"),(Type=None,Name="None"),(Type=None,Name="None"))
     Element=@n34.Element
     Array.0=RigVMVariableNode() {variable=IK Control}.Value
     Array.1=RigVMVariableNode() {variable=PV Control}.Value
     Array.2=RigVMVariableNode() {variable=IK Roll Control}.Value
    ) {template="DISPATCH_RigVMDispatch_ArrayFind(in Array,in Element,out Index,out Success)"}.Success
    ExecuteContext=@n35.False
   )
   RigVMVariableNode(Value=true, ExecuteContext=@n36.True) {variable=IK Solve}
   RigVMVariableNode(ExecuteContext=@n35.True) {variable=IK Solve}
   Hide Show Controls_1_1 = RigVMFunctionReferenceNode(
    IK Controls=RigVMDispatchNode(
     Values=((Type=None,Name="None"),((),Type=None,Name="None"),((),Type=None,Name="None"))
     Values.0=RigVMVariableNode() {variable=IK Control}.Value
     Values.2=RigVMVariableNode() {variable=PV Control}.Value
     Values.1=RigVMVariableNode() {variable=IK Roll Control}.Value
    ) {template="DISPATCH_RigVMDispatch_ArrayMake(in Values,out Array)"}.Array
    FK Controls=RigVMVariableNode() {variable=FK Controls}.Value
    IK Solve=RigVMVariableNode() {variable=IK Solve}.Value
    ExecuteContext=@n31.False
   )
   n37 = /Script/RigVM.RigVMFunction_UserDefinedEvent(EventName=To IK) {is=struct RigVMUnitNode}
   Match IK Controls_1 = RigVMFunctionReferenceNode(ExecuteContext=@n37.ExecutePin)
   n38 = &RigVMFunction_UserDefinedEvent(EventName=To FK)
   Match FK Controls_0 = RigVMFunctionReferenceNode(ExecuteContext=@n38.ExecutePin)
  ]}
 RigVMModel Rig {nodes [
   n1 = /Script/ControlRig.RigUnit_InverseExecution() {is=struct RigVMUnitNode}
   Match FK Controls = RigVMFunctionReferenceNode(ExecuteContext=@n1.ExecutePin)
   > Match IK Controls = RigVMFunctionReferenceNode()
  ]}
]
hierarchy [
 Bone A {type=Connector, loc=<0 0 0>, rot=<0 0 0 1>}
 Root {type=Socket, loc=<0 0 0>, rot=<0 0 0 1>}
 Bone B {type=Connector, loc=<0 0 0>, rot=<0 0 0 1>}
 Bone C {type=Connector, loc=<0 0 0>, rot=<0 0 0 1>}
 Bone D {type=Connector, loc=<0 0 0>, rot=<0 0 0 1>}
 Bone E {type=Connector, loc=<0 0 0>, rot=<0 0 0 1>}
 IK Parent {type=Connector, loc=<0 0 0>, rot=<0 0 0 1>}
 FK Orient Space {type=Connector, loc=<0 0 0>, rot=<0 0 0 1>}
 FK Parent {type=Connector, loc=<0 0 0>, rot=<0 0 0 1>}
]
variables [
 Bone A {type=FRigElementKey, default=(Type=None,Name="")}
 Bone B {type=FRigElementKey, default=(Type=None,Name="")}
 Bone C {type=FRigElementKey, default=(Type=None,Name="")}
 Bone D {type=FRigElementKey, default=(Type=None,Name="")}
 Color {type=FLinearColor, default=(R=1,G=1,B=1,A=1)}
 Control Scale {type=double, default=1}
 Draw Pv Channel {type=FRigElementKey, default=(Type=None,Name="")}
 Effector Location {type=FVector, default=(X=0,Y=0,Z=0)}
 Effector Offset {type=FVector, default=(X=0,Y=0,Z=0)}
 End Roll Control {type=bool, default=True}
 FK Bones {type="TArray<FRigElementKey>", default=(())}
 FK Controls {type="TArray<FRigElementKey>", default=(())}
 FK Orient Space Null {type=FRigElementKey, default=(Type=None,Name="")}
 FK Parent {type=FRigElementKey, default=(Type=None,Name="")}
 FK Parent Connector {type=FRigElementKey, default=(Type=None,Name="")}
 FKs Control Shape {
  type=FRigUnit_HierarchyAddControl_ShapeSettings
  default=(bVisible=True,Name="Default",Color=(R=1,G=1,B=1,A=1),Transform=(Rotation=(X=0,Y=0,Z=0,W=1),Translation=(X=0,Y=0,Z=0),Scale3D=(X=1,Y=1,Z=1)))
 }
 Fk Controls Rotation Offset {type=FQuat, default=(X=0,Y=0,Z=0,W=1)}
 IK Control {type=FRigElementKey, default=(Type=None,Name="")}
 IK Control Shape {
  type=FRigUnit_HierarchyAddControl_ShapeSettings
  default=(bVisible=True,Name="Default",Color=(R=1,G=1,B=1,A=1),Transform=(Rotation=(X=0,Y=0,Z=0,W=1),Translation=(X=0,Y=0,Z=0),Scale3D=(X=1,Y=1,Z=1)))
 }
 IK Effector Null {type=FRigElementKey, default=(Type=None,Name="")}
 IK FK Channel {type=FRigElementKey, default=(Type=None,Name="")}
 IK FK Manipulation {type=bool, default=False}
 IK Parent {type=FRigElementKey, default=(Type=Bone,Name="")}
 IK Roll Aim Null {type=FRigElementKey, default=(Type=None,Name="")}
 IK Roll Control {type=FRigElementKey, default=(Type=None,Name="")}
 IK Roll Control Shape {
  type=FRigUnit_HierarchyAddControl_ShapeSettings
  default=(bVisible=True,Name="Hexagon_Thick",Color=(R=1,G=1,B=1,A=1),Transform=(Rotation=(X=0,Y=0,Z=0,W=1),Translation=(X=0,Y=0,Z=0),Scale3D=(X=1,Y=1,Z=1)))
 }
 IK Roll Offset Null {type=FRigElementKey, default=(Type=Bone,Name="")}
 IK Solve {type=bool, default=False}
 Length {type=double, default=0}
 Mirror Axis {type=FVector, default=(X=0,Y=1,Z=1)}
 Mirror Behavior {type=bool, default=False}
 PV Control {type=FRigElementKey, default=(Type=None,Name="")}
 PV Control Color {type=FLinearColor, default=(R=1,G=1,B=1,A=1)}
 PV Control Shape {
  type=FRigUnit_HierarchyAddControl_ShapeSettings
  default=(bVisible=True,Name="Default",Color=(R=1,G=1,B=1,A=1),Transform=(Rotation=(X=0,Y=0,Z=0,W=1),Translation=(X=0,Y=0,Z=0),Scale3D=(X=1,Y=1,Z=1)))
 }
 PV Distance Mult {type=double, default=1}
 PV Root Space Null {type=FRigElementKey, default=(Type=None,Name="")}
 PV Twist Follow {type=double, default=1}
 PV Twist Follow Channel {type=FRigElementKey, default=(Type=None,Name="")}
 Previous IK Solve {type=bool, default=False}
 Primary Axis {type=FVector, default=(X=1,Y=0,Z=0)}
 Rest Distance {type=double, default=0}
 Rotation Offset {type=FQuat, default=(X=0,Y=0,Z=0,W=1)}
 Secondary Axis {type=FVector, default=(X=0,Y=0,Z=-1)}
]
